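# Translucent unlit Niagara ribbon material for the Toony Beam vehicle boost cosmetic, drawing a scrolling, distorted beam.
# A panning normal map, scaled by dynamic parameter AddCore, warps texture and mask UVs; particle color tints the texture through an emissive adjustment function.
# Opacity subtracts panning noise erosion, driven by Param4 and a U-axis tail mask, from texture and particle alpha.
# A second noise layer breaks up the V edges before depth fade.
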Material /VehicleCosmetics/Boosts/Boost_ToonyBeam/Materials/MM_Boost_Toony_Beam_Ribbon
graph {
 properties {
  bAutomaticallySetUsageInEditor=false
  bUsedWithNiagaraRibbons=true
  blendMode=BLEND_Translucent
  shadingModel=MSM_Unlit
  usedShadingModels=Unlit
 }
 nodes [
  MaterialExpressionVertexColor_2 = ParticleColor()
  n1 = TextureCoordinate()
  n2 = TextureCoordinate()
  n3 = Time()
  n4 = Panner(
   Coordinate=Multiply(A=ScalarParameter() {defaultValue=1, group=Distortion, parameterName=Distortion_Scale_X}, B=@n2)
   Time=Multiply(A=ScalarParameter() {defaultValue=-1, group=Distortion, parameterName=Distortion_Speed_X}, B=@n3)
  ) {speedX=1}
  n5 = Panner(
   Coordinate=Multiply(A=@n2, B=ScalarParameter() {defaultValue=1, group=Distortion, parameterName=Distortion_Scale_Y})
   Time=Multiply(A=@n3, B=ScalarParameter() {defaultValue=-1, group=Distortion, parameterName=Distortion_Speed_Y})
  ) {speedY=1}
  MaterialExpressionTextureSampleParameterNormal_3 = TextureSampleParameter2D(UVs=AppendVector(A=ComponentMask(None=@n4) {r=true}, B=ComponentMask(None=@n5) {g=true})) {
   group=Distortion
   parameterName=Normal_Map
   samplerType=SAMPLERTYPE_Normal
   texture=/DealershipSharedLibrary/FX/Textures/Noise/T_VCVFX_CloudNoise_N
  }
  n6 = Multiply(
   A=DynamicParameter() {paramNames [AddCore,AddMid,AddOut,Param4]}.AddCore
   B=ComponentMask(None=@MaterialExpressionTextureSampleParameterNormal_3.RGB) {g=true, r=true}
  )
  n7 = ScalarParameter() {defaultValue=1, group=Texture, parameterName=Texture_Scale_X}
  n8 = TextureCoordinate()
  n9 = ScalarParameter() {defaultValue=-1, group=Texture, parameterName=Texture_Speed_X}
  n10 = Time()
  n11 = ComponentMask(None=Panner(Coordinate=Multiply(A=@n7, B=@n8), Time=Multiply(A=@n9, B=@n10)) {speedX=1}) {r=true}
  n12 = ScalarParameter() {defaultValue=1, group=Texture, parameterName=Texture_Scale_Y}
  n13 = ScalarParameter() {defaultValue=-1, group=Texture, parameterName=Texture_Speed_Y}
  n14 = ComponentMask(None=Panner(Coordinate=Multiply(A=@n8, B=@n12), Time=Multiply(A=@n10, B=@n13)) {speedY=1}) {g=true}
  n15 = TextureSampleParameter2D(UVs=Add(A=Add(A=@n1, B=@n6), B=AppendVector(A=@n11, B=@n14))) {group=Texture, parameterName=Texture_Map, texture=/VehicleCosmetics/Boosts/Boost_ToonyBeam/Textures/T_Boost_Toony_Beam_Ribbon_01}
  n16 = TextureCoordinate()
  n17 = Time()
  n18 = Panner(
   Coordinate=Multiply(A=ScalarParameter() {defaultValue=1, group=Threshold, parameterName=Threshold_Scale_X}, B=@n16)
   Time=Multiply(A=ScalarParameter() {defaultValue=-1, group=Threshold, parameterName=Threshold_Speed_X}, B=@n17)
  ) {speedX=1}
  n19 = Panner(
   Coordinate=Multiply(A=@n16, B=ScalarParameter() {defaultValue=1, group=Threshold, parameterName=Threshold_Scale_Y})
   Time=Multiply(A=@n17, B=ScalarParameter() {defaultValue=-1, group=Threshold, parameterName=Threshold_Speed_Y})
  ) {speedY=1}
  n20 = TextureSampleParameter2D(UVs=AppendVector(A=ComponentMask(None=@n18) {r=true}, B=ComponentMask(None=@n19) {g=true})) {
   group=Threshold
   parameterName=Threshold_Map
   sortPriority=0
   texture=/DealershipSharedLibrary/FX/Textures/Gradient_Masks/T_VCVFX_Blob_Noise_A
  }
  n21 = Add(A=@n15.A, B=@n20.RGB)
  n22 = DynamicParameter() {paramNames [AddCore,AddMid,AddOut,Param4]}
  n23 = ComponentMask(None=TextureCoordinate()) {r=true}
  n24 = ScalarParameter() {defaultValue=2, group=Mask, parameterName=Mask_Power_Tail}
  n25 = Subtract(
   A=Multiply(A=Power(Base=@n23, Exp=@n24), B=ScalarParameter() {defaultValue=20, group=Mask, parameterName=Mask_Hardness_Tail})
   B=Constant() {r=-5}
  )
  n26 = ScalarParameter() {defaultValue=1, group=Mask, parameterName=Mask_Scale_X}
  n27 = TextureCoordinate()
  n28 = ScalarParameter() {defaultValue=-1, group=Mask, parameterName=Mask_Speed_X}
  n29 = Time()
  n30 = ComponentMask(None=Panner(Coordinate=Multiply(A=@n26, B=@n27), Time=Multiply(A=@n28, B=@n29)) {speedX=1}) {r=true}
  n31 = ScalarParameter() {defaultValue=1, group=Mask, parameterName=Mask_Scale_Y}
  n32 = ScalarParameter() {group=Mask, parameterName=Mask_Speed_Y}
  n33 = ComponentMask(None=Panner(Coordinate=Multiply(A=@n27, B=@n31), Time=Multiply(A=@n29, B=@n32)) {speedY=1}) {g=true}
  n34 = TextureCoordinate()
  n35 = TextureCoordinate()
  n36 = Time()
  n37 = Panner(
   Coordinate=Multiply(A=ScalarParameter() {defaultValue=1, group=Distortion, parameterName=Distortion_Scale_X}, B=@n35)
   Time=Multiply(A=ScalarParameter() {defaultValue=-1, group=Distortion, parameterName=Distortion_Speed_X}, B=@n36)
  ) {speedX=1}
  n38 = Panner(
   Coordinate=Multiply(A=@n35, B=ScalarParameter() {defaultValue=1, group=Distortion, parameterName=Distortion_Scale_Y})
   Time=Multiply(A=@n36, B=ScalarParameter() {defaultValue=-1, group=Distortion, parameterName=Distortion_Speed_Y})
  ) {speedY=1}
  MaterialExpressionTextureSampleParameterNormal_4 = TextureSampleParameter2D(UVs=AppendVector(A=ComponentMask(None=@n37) {r=true}, B=ComponentMask(None=@n38) {g=true})) {
   group=Distortion
   parameterName=Normal_Map
   samplerType=SAMPLERTYPE_Normal
   texture=&T_VCVFX_CloudNoise_N
  }
  n39 = Multiply(
   A=DynamicParameter() {paramNames [AddCore,AddMid,AddOut,Param4]}.AddCore
   B=ComponentMask(None=@MaterialExpressionTextureSampleParameterNormal_4.RGB) {g=true, r=true}
  )
  n40 = TextureSampleParameter2D(UVs=Add(A=AppendVector(A=@n30, B=@n33), B=Add(A=@n34, B=@n39))) {group=Mask, parameterName=Texture_Mask, texture=&T_VCVFX_Blob_Noise_A}
  n41 = Subtract(A=Multiply(A=@MaterialExpressionVertexColor_2.A, B=@n15.A), B=Multiply(A=Multiply(A=@n21, B=@n22.Param4), B=Multiply(A=@n25, B=@n40.RGB)))
  n42 = TextureCoordinate()
  n43 = TextureCoordinate()
  n44 = Time()
  n45 = Panner(Coordinate=Multiply(A=Constant() {r=2}, B=@n43), Time=Multiply(A=Constant() {r=-3}, B=@n44)) {speedX=1}
  n46 = Panner(Coordinate=Multiply(A=@n43, B=Constant() {r=2}), Time=Multiply(A=@n44, B=Constant())) {speedY=1}
  n47 = TextureSampleParameter2D(UVs=AppendVector(A=ComponentMask(None=@n45) {r=true}, B=ComponentMask(None=@n46) {g=true})) {
   group=Distortion
   parameterName=Normal_Map
   samplerType=SAMPLERTYPE_Normal
   texture=&T_VCVFX_CloudNoise_N
  }
  n48 = Multiply(A=DynamicParameter() {paramNames [AddCore,AddMid,AddOut,Param4]}.AddCore, B=ComponentMask(None=@n47.RGB) {g=true, r=true})
  n49 = Constant() {r=10}
  n50 = TextureCoordinate()
  n51 = Constant() {r=-3}
  n52 = Time()
  n53 = ComponentMask(None=Panner(Coordinate=Multiply(A=@n49, B=@n50), Time=Multiply(A=@n51, B=@n52)) {speedX=1}) {r=true}
  n54 = Constant() {r=1}
  n55 = Constant()
  n56 = ComponentMask(None=Panner(Coordinate=Multiply(A=@n50, B=@n54), Time=Multiply(A=@n52, B=@n55)) {speedY=1}) {g=true}
  n57 = TextureSampleParameter2D(UVs=Add(A=Add(A=@n42, B=@n48), B=AppendVector(A=@n53, B=@n56))) {
   group=Threshold
   parameterName=Threshold_Map
   sortPriority=0
   texture=&T_VCVFX_Blob_Noise_A
  }
  n58 = Constant() {r=200}
  n59 = TextureCoordinate()
  n60 = Power(Base=ComponentMask(None=@n59) {g=true}, Exp=Constant() {r=5.25})
  n61 = Power(Base=OneMinus(None=ComponentMask(None=@n59) {g=true}), Exp=Constant() {r=5.2})
  n62 = Constant() {r=200}
  MaterialExpressionConstantClamp_5 = Saturate(None=Multiply(A=Multiply(A=@n58, B=@n60), B=Multiply(A=@n61, B=@n62)))
  n63 = Multiply(A=@n57.RGB, B=OneMinus(None=@MaterialExpressionConstantClamp_5))
  n64 = Constant() {r=1}
  n65 = Constant() {r=100}
  MaterialExpressionConstantClamp_4 = Saturate(None=Multiply(
    A=Subtract(A=@n41, B=@n63)
    B=MaterialFunctionCall(Opacity In=@n64, DepthFade Distance=@n65) {materialFunction=&MF_VCVFX_DepthFade}.Result
   ))
  MaterialFunctionCall(Emissive=Multiply(A=@MaterialExpressionVertexColor_2.RGB, B=@n15.RGB)) {materialFunction=&MF_FX_EmissiveAdjustments}
 ]
}
